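# Translucent HUD UI material that draws scrolling layers of triangles over a glowing, flame-colored flare.
# Three samples of one triangle texture, each using its own channel, scale and offset, pan horizontally at different speeds in aspect-corrected UVs.
# SmoothStep thresholds that vary along U shape each layer before nested lerps combine them into a mask.
# An exponential radial gradient forms the flare, and the mask blends its colors toward a vertical triangle tint.
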Material /Game/UI/HUD/Materials/M_FX_TriangleSpiral
graph {
 properties {
  bUsedWithStaticMesh=false
  blendMode=BLEND_Translucent
  materialDomain=MD_UI
  pixelDepthOffsetMode=PDOM_Legacy
  usedShadingModels=Default Lit
 }
 nodes [
  n1 = MaterialFunctionCall() {materialFunction=&GetUserInterfaceUV}
  n2 = MaterialFunctionCall(Float2=@n1.Pixel Size) {materialFunction=&BreakOutFloat2Components}
  n3 = AppendVector(A=Divide(A=@n2.R, B=Max(B=@n2.G) {constA=0.1}), B=Constant() {r=1})
  n4 = ScalarParameter() {defaultValue=0.9048, parameterName=tri-L3-Scale}
  n5 = Add(
   A=Multiply(A=@n1.Normalized UV, B=Multiply(A=@n3, B=@n4))
   B=AppendVector(A=Constant(), B=ScalarParameter() {defaultValue=0.019, parameterName=tri-L3-Offset})
  )
  n6 = TextureObjectParameter() {parameterName=Texture, samplerType=SAMPLERTYPE_LinearColor, texture=/Game/UI/Foundation/Items/Textures/T-Card-Hero-Anim-Triangle}
  n7 = ScalarParameter() {defaultValue=1.95, parameterName=ss min}
  n8 = Power(Base=ComponentMask(None=TextureCoordinate()) {r=true}) {constExponent=1.7}
  n9 = ScalarParameter() {defaultValue=0.35, parameterName=ss max}
  n10 = MaterialFunctionCall(
   Alpha=TextureSample(UVs=Panner(Coordinate=@n5) {speedX=-0.35}, Tex=@n6) {samplerType=SAMPLERTYPE_LinearColor, texture=&T-Card-Hero-Anim-Triangle}.B
   Min=LinearInterpolate(B=@n7, Alpha=@n8) {constA=0.25, constB=0.8}
   Max=LinearInterpolate(A=@n9, Alpha=Multiply(A=@n8)) {constA=0.2, constB=2}
  ) {materialFunction=&SmoothStep}
  n11 = Multiply(A=@n1.Normalized UV, B=Multiply(A=@n3, B=ScalarParameter() {defaultValue=0.5743, parameterName=tri-L2-Scale}))
  n12 = AppendVector(A=Constant(), B=ScalarParameter() {defaultValue=-0.6765, parameterName=tri-L2-Offset})
  n13 = TextureSample(UVs=Panner(Coordinate=Add(A=@n11, B=@n12)) {speedX=-0.45}, Tex=@n6) {samplerType=SAMPLERTYPE_LinearColor, texture=&T-Card-Hero-Anim-Triangle}
  n14 = LinearInterpolate(B=@n7, Alpha=@n8) {constA=0.25, constB=0.9}
  n15 = LinearInterpolate(A=@n9, Alpha=@n8) {constA=0.1, constB=2}
  n16 = Multiply(A=@n3, B=ScalarParameter() {defaultValue=0.33, parameterName=tri-L1-Scale})
  n17 = Constant()
  n18 = ScalarParameter() {defaultValue=0.2476, parameterName=tri-L1-Offset}
  n19 = Panner(Coordinate=Add(A=Multiply(A=@n1.Normalized UV, B=@n16), B=AppendVector(A=@n17, B=@n18))) {speedX=-0.5}
  n20 = LinearInterpolate(
   A=LinearInterpolate(
    A=Multiply(A=@n10.Result)
    Alpha=MaterialFunctionCall(Alpha=@n13.G, Min=@n14, Max=@n15) {materialFunction=&SmoothStep}.Result
   ) {constB=0.25}
   Alpha=MaterialFunctionCall(
    Alpha=TextureSample(UVs=@n19, Tex=@n6) {samplerType=SAMPLERTYPE_LinearColor, texture=&T-Card-Hero-Anim-Triangle}.R
    Min=LinearInterpolate(B=@n7, Alpha=@n8) {constA=0.25, constB=0.9}
    Max=LinearInterpolate(A=@n9, Alpha=@n8) {constA=0.1, constB=2}
   ) {materialFunction=&SmoothStep}.Result
  )
  n21 = TextureCoordinate() {uTiling=1.5, vTiling=1.5}
  n22 = Constant2Vector() {g=-0.5}
  n23 = MaterialFunctionCall(
   UVs=MaterialFunctionCall(UVs=Add(A=@n21, B=@n22), Texture Scale=Constant2Vector() {g=1, r=1.65}) {materialFunction=&ScaleUVsByCenter}.UVs
   CenterPosition=VectorParameter() {parameterName=FlareCenter, defaultValue=<rgba 0 0.25 0 1>}.RGB
   Radius=ScalarParameter() {defaultValue=0.85, parameterName=FlareRadius}
   Density=ScalarParameter() {defaultValue=1.75, parameterName=FlareDensity}
  ) {materialFunction=&RadialGradientExponential}
  Add(A=@n23.RadialGradientExponential, B=@n20)
  ComponentMask(None=TextureCoordinate()) {r=true}
  ComponentMask(None=TextureCoordinate()) {r=true}
  n24 = VectorParameter() {parameterName=RGB_OuterFlame, defaultValue=<rgba 0 0.25 1 1>}
  n25 = VectorParameter() {parameterName=RGB_InnerFlame, defaultValue=<rgba 0 1 1 1>}
  n26 = Power(Base=@n23.RadialGradientExponential) {constExponent=0.85}
  n27 = ComponentMask(None=TextureCoordinate()) {g=true}
  LinearInterpolate(
   A=LinearInterpolate(
    A=LinearInterpolate(A=@n24.RGB, B=@n25.RGB, Alpha=@n26)
    B=Constant3Vector() {constant=<rgba 1 1 1 0>}
    Alpha=Power(Base=@n23.RadialGradientExponential) {constExponent=4}
   )
   B=LinearInterpolate(
    A=VectorParameter() {parameterName=Triangle_RGB_B, defaultValue=<rgba 1 1 1 1>}.RGB
    B=VectorParameter() {parameterName=Triangle_RGB_A, defaultValue=<rgba 0 1 1 1>}.RGB
    Alpha=Power(Base=@n27) {constExponent=0.35}
   )
   Alpha=@n20
  )
  Round(None=@n20)
 ]
}
